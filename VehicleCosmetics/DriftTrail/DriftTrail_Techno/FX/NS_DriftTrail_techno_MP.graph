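# Niagara system for the Techno drift trail vehicle cosmetic, combining wheel-wrap meshes with plasma energy ribbons.
# The wrap emitter bursts particles partitioned onto wheel bones and scaled per wheel, while TrailPlasmaEnergy spawns two backward-moving ribbons with sine-offset ray positions and animated widths.
# Both emitters mask disabled tires and apply painted color variants with per-renderer overrides, animated by age on the ribbons.
# A frontend-preview flag changes ribbon lifetime, offset and velocity.
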NiagaraSystem /VehicleCosmetics/DriftTrail/DriftTrail_Techno/FX/NS_DriftTrail_techno_MP
dependencies {
 usedDataInterfaces [&NiagaraDataInterfaceCurve, &NiagaraDataInterfaceColorCurve, &NiagaraDataInterfaceSkeletalMesh]
 usedDynamicInputs []
 usedModules [
  &self:SystemSpawnScript.SystemScriptSource.SystemScriptGraph.NiagaraNodeAssignment_0.SetVariables_#1
  &SystemState
  &FN_UnpackVehicleFlags
  /Game/Effects/Niagara/Modules/Vehicle/FN_UpdateVehicleAttribs
  /Game/Effects/Niagara/Modules/Vehicle/FN_VehicleFlags_Cheatcodes
 ]
}
emitters {
 NE_VC_DriftTrail_WheelWrap_Parent {
  emitterUpdateScript [
   &EmitterState(
    Inactive Response=::NewEnumerator1
    Life Cycle Mode=::NewEnumerator1
    Loop Behavior=::NewEnumerator1
    Loop Duration Mode=::NewEnumerator1
    UseLoopDelay=false
   )
   &SpawnBurst_Instantaneous(Spawn Count=8, Use Spawn Probability=false, UseLoopCountLimit=false)
  ]
  particleSpawnScript [
   &VC_PaintedVariantsOverride(
    Black=<rgba 0.06 0.06 0.06 5>
    BurntSienna=<rgba 0.3 0.1108 0.06 10>
    Cobalt=<rgba 0 0.037 2 20>
    Crimson=<rgba 0.4635 0.1391 0.1479 20>
    DefaultColor=<rgba 0.103 0.309 0.5149 0.3>
    ForestGreen=<rgba 0.075 0.3 0.03 20>
    Grey=<rgba 0.1 0.1 0.1 20>
    Lime=<rgba 0.1667 0.5 0 20>
    Orange=<rgba 0.5 0.1744 0 20>
    Pink=<rgba 0.5052 0.2021 0.395 20>
    Purple=<rgba 0.2083 0.15 0.5 20>
    R1 Black:LinearColor=<rgba 0 0 0 50>
    R1 BurntSienna:LinearColor=<rgba 0.03 0.0093 0.0037 10>
    R1 Cobalt:LinearColor=<rgba 0.002 0.0038 0.02 20>
    R1 Crimson:LinearColor=<rgba 0.02 0 0.0005 20>
    R1 DefaultColor=<rgba 0.2135 5 4.6011 1>
    R1 ForestGreen:LinearColor=<rgba 0 0 0 20>
    R1 Grey:LinearColor=<rgba 0.01 0.01 0.01 20>
    R1 Lime:LinearColor=<rgba 0.0067 0.02 0 20>
    R1 Orange:LinearColor=<rgba 0.03 0.0105 0 20>
    R1 Pink:LinearColor=<rgba 0.02 0.0024 0.0136 20>
    R1 Purple:LinearColor=<rgba 0.0041 0 0.01 20>
    R1 Saffron:LinearColor=<rgba 0.0188 0.02 0.0013 20>
    R1 SkyBlue:LinearColor=<rgba 0 0.0067 0.01 20>
    R1 TitaniumWhite:LinearColor=<rgba 0.05 0.05 0.05 20>
    R2 Black:LinearColor=<rgba 0.03 0.03 0.03 0.5>
    R2 BurntSienna:LinearColor=<rgba 0.4 0.1235 0.0493 1>
    R2 Cobalt:LinearColor=<rgba 0.12 0.7667 4 0.1>
    R2 Crimson:LinearColor=<rgba 7 0 0.1167 0.2>
    R2 DefaultColor:LinearColor=<rgba 0.0854 2 1.8404 0.5>
    R2 ForestGreen:LinearColor=<rgba 0.1 0.5 0.1202 1>
    R2 Grey:LinearColor=<rgba 0.5 0.5 0.5 0.6>
    R2 Lime:LinearColor=<rgba 0.6 1.2 0 0.5>
    R2 Orange:LinearColor=<rgba 2 0.8278 0.2 0.1>
    R2 Pink:LinearColor=<rgba 4 0.4792 2.7197 1>
    R2 Purple:LinearColor=<rgba 0.2453 0.16 0.8 0.2>
    R2 Saffron:LinearColor=<rgba 1.455 1.5 0.15 0.3>
    R2 SkyBlue:LinearColor=<rgba 0 1 3 0.5>
    R2 TitaniumWhite:LinearColor=<rgba 2 2 2 0.4>
    R3 Black:LinearColor=<rgba 0 0 0 1>
    R3 BurntSienna:LinearColor=<rgba 0.04 0.0124 0.0049 1>
    R3 Cobalt:LinearColor=<rgba 0 0.0833 1 1>
    R3 Crimson:LinearColor=<rgba 0.4 0.0015 0.0041 1>
    R3 DefaultColor:LinearColor=<rgba 0.02 0.0333 0.1 1>
    R3 ForestGreen:LinearColor=<rgba 0 0.07 0.0035 1>
    R3 Grey:LinearColor=<rgba 0.03 0.03 0.03 1>
    R3 Lime:LinearColor=<rgba 0.2 0.4 0 0.8>
    R3 Orange:LinearColor=<rgba 0.55 0.0667 0.0029 1>
    R3 Pink:LinearColor=<rgba 0.5 0.15 0.3727 0.9>
    R3 Purple:LinearColor=<rgba 0.0587 0 0.5029 1>
    R3 Saffron:LinearColor=<rgba 0.2292 0.25 0 1>
    R3 SkyBlue:LinearColor=<rgba 0.03 0.1425 0.3 1>
    R3 TitaniumWhite:LinearColor=<rgba 0.6 0.6 0.6 1>
    Saffron=<rgba 0.6 0.55 0 20>
    SkyBlue=<rgba 0.05 0.275 0.5 20>
    TitaniumWhite=<rgba 0.9 0.9 0.9 20>
    Use R1 Painted Overrides=true
    Use R2 Painted Overrides=true
    Use R3 Painted Overrides=true
   )
   &InitializeParticle(
    Color=$Particles.Color
    Lifetime=$Emitter.CurrentLoopDuration
    Mesh Scale Mode=::NewEnumerator4
    Position Offset=<150 150 0>
    Sprite Rotation Mode=::NewEnumerator2
    Sprite Size Mode=::NewEnumerator3
    Uniform Sprite Size=&VectorLength(Vector=<10 10 10>)
    UsePositionOffset=false
    Write Material Random=true
   )
   &PartitionParticles()
   &SkeletalMeshLocation(
    Bone / Socket Index=&Add_Integer(A=$Particles.PartitionParticles.Partition, B=4)
    Bone / Socket Sampling Mode=::NewEnumerator5
    Interpolate Mesh Sampling=false
    MaskSpawnGroup=false
    Mesh Sampling Type=::NewEnumerator6
    Orientation Sampling=::NewEnumerator1
    Skeletal Mesh=$System.SkeletalMesh
   )
  ]
  particleUpdateScript [
   &FN_DisableTireFX(
    Mask Index Attribute:NiagaraInt32=$Particles.PartitionParticles.Partition
    Mask Index FL=2
    Mask Index FR=3
    Mask Index RL=0
    Mask Index RR=1
   )
   &VC_WheelWrapScale(Wheel Index Attribute:NiagaraInt32=&Add_Integer(A=$Particles.PartitionParticles.Partition, B=2))
   &ParticleState(Kill Particles When Lifetime Has Elapsed=false, Loop Particles Lifetime=true)
   &UpdateMeshOrientation(Rotation Coordinate Space=::NewEnumerator2, Rotation Vector=<0 1 0>) {enabled=false}
   &SkeletalMeshLocation(
    Bone / Socket Index=&Add_Integer(A=$Particles.PartitionParticles.Partition, B=4)
    Bone / Socket Sampling Mode=::NewEnumerator5
    MaskSpawnGroup=false
    Mesh Sampling Type=::NewEnumerator6
    Skeletal Mesh=$System.SkeletalMesh
   ) {module=SkeletalMeshLocation001}
   &SolveForcesAndVelocity(Clamp Velocity=false, Limit Acceleration=false)
  ]
  rendererClasses [&NiagaraMeshRendererProperties]
  renderers [
   {rendererClass=&NiagaraMeshRendererProperties, rendererIndex=0}
   {rendererClass=&NiagaraMeshRendererProperties, rendererIndex=1}
   {rendererClass=&NiagaraMeshRendererProperties, rendererIndex=2}
   {rendererClass=&NiagaraMeshRendererProperties, rendererIndex=3}
   {rendererClass=&NiagaraMeshRendererProperties, rendererIndex=4}
  ]
 }
 TrailPlasmaEnergy {
  emitterUpdateScript [
   &EmitterState(Inactive Response=::NewEnumerator1, Life Cycle Mode=::NewEnumerator1, Loop Duration=1, UseLoopDelay=false)
   &VC_RateQualityHandler(Epic Spawn Rate=80, High Spawn Rate=68, Low Spawn Rate=50, Medium Spawn Rate=60, Use Custom Rate?=true) {enabled=false, module=VC_RateQualityHandler001}
   &VC_RateQualityHandler(Epic Spawn Rate=90, Low Spawn Rate=30, Medium Spawn Rate=45, Use Custom Rate?=true)
   &VC_SpawnRibbon_v1(Number of Ribbons=2, Spawn Time=-1, SpawnRate=$Emitter.Spawn Rate)
  ]
  particleSpawnScript [
   &VC_InitializeRibbon_v1(
    Color=$Particles.Color
    Custom Offset=true
    Enable Skeletal Position=true
    Lifetime=&MakeCustomFloatFromBool(Bool=$User.bIsFrontendPreview, True Float=0.6, False Float=0.8)
    Local Vector Offset=&MakeCustomVectorFromBool(
     Bool=$User.bIsFrontendPreview
     True Vector=&AddVectorToPosition(Position=$User.offset001, Vector=<-40 0 -12>)
     False Vector=&AddVectorToPosition(Position=$User.offset001, Vector=<-40 0 -25>)
    )
    Offset Amount=1
    Ribbon Width=150
   )
   &AddVelocity(Velocity=&MakeCustomVectorFromBool(Bool=$User.bIsFrontendPreview, True Vector=<-1600 0 0>, False Vector=<-1000 0 0>)) {module=AddVelocity001}
  ]
  particleUpdateScript [
   &FN_DisableTireFX(Mask Index Attribute:NiagaraInt32=$Particles.SpawnGroup, Mask Index FL=2, Mask Index FR=3, Mask Index RL=0, Mask Index RR=1)
   &InheritSourceMovement(
    Applied Position Scale=&VectorFromFloat(Value=&FloatFromCurve(FloatCurve=DataInterface() {propertyValues {Curve {keys [
          {interpMode=RCIM_Cubic, tangentMode=RCTM_Break, time=0, value=1}
          {arriveTangent=-2.9025, interpMode=RCIM_Cubic, tangentMode=RCTM_Break, time=1, value=0}
         ]}}}))
    ApplyForce=false
    ApplyPosition=true
    ApplyPositionFalloff=false
   )
   &ParticleState()
   &SolveForcesAndVelocity(Clamp Velocity=false, Limit Acceleration=false)
   &VC_UpdateRibbon_v1(Enable Skeletal Position=true)
   &VC_PaintedVariantsOverride(
    Black=&ColorFromCurve(ColorCurve=DataInterface() {propertyValues {
       AlphaCurve {keys [<at 0 5 RCIM_Cubic>,<at 0.7275 0 RCIM_Cubic>]}
       BlueCurve {keys [<at 0 0.05 RCIM_Cubic>]}
       GreenCurve {keys [<at 0 0.05 RCIM_Cubic>]}
       RedCurve {keys [<at 0 0.05 RCIM_Cubic>]}
      }})
    BurntSienna=&ColorFromCurve(ColorCurve=DataInterface() {propertyValues {
       AlphaCurve {keys [<at 0 2 RCIM_Cubic>,<at 0.7743 0 RCIM_Cubic>]}
       BlueCurve {keys [<at 0 0.0083 RCIM_Cubic>]}
       GreenCurve {keys [<at 0 0.1128 RCIM_Cubic>]}
       RedCurve {keys [<at 0 0.29 RCIM_Cubic>]}
      }})
    Cobalt=&ColorFromCurve(ColorCurve=DataInterface() {propertyValues {
       AlphaCurve {keys [<at 0 2 RCIM_Cubic>,<at 0.7616 0>]}
       BlueCurve {keys [<at 0 1 RCIM_Cubic>,<at 0.7634 0.05>]}
       GreenCurve {keys [<at 0 0.0185 RCIM_Cubic>,<at 0.7634 0.0014>]}
       RedCurve {keys [<at 0 0 RCIM_Cubic>,<at 0.7634 0.0001>]}
      }})
    Crimson=&ColorFromCurve(ColorCurve=DataInterface() {propertyValues {
       AlphaCurve {keys [<at 0 2 RCIM_Cubic>,<at 0.7795 0 RCIM_Cubic>]}
       BlueCurve {keys [<at 0 0.0266 RCIM_Cubic>,<at 0.7849 0>]}
       GreenCurve {keys [<at 0 0 RCIM_Cubic>,<at 0.7849 0>]}
       RedCurve {keys [<at 0 1 RCIM_Cubic>,<at 0.7849 0>]}
      }})
    DefaultColor=&ColorFromCurve(ColorCurve=DataInterface() {propertyValues {
       AlphaCurve {keys [<at 0 1.2 RCIM_Cubic>,<at 0.4627 0.6>,<at 0.7329 -0.2 RCIM_Cubic>]}
       BlueCurve {keys [{arriveTangent=-0.789, interpMode=RCIM_Cubic, time=0, value=0.24}, <at 0.4617 0.0685>]}
       GreenCurve {keys [{arriveTangent=-0.789, interpMode=RCIM_Cubic, time=0, value=0.3}, <at 0.4617 0.0409>]}
       RedCurve {keys [{arriveTangent=-4.4991, interpMode=RCIM_Cubic, time=0, value=0.06}, <at 0.4617 0.0037>]}
      }})
    ForestGreen=&ColorFromCurve(ColorCurve=DataInterface() {propertyValues {
       AlphaCurve {keys [<at 0 4 RCIM_Cubic>,<at 0.4854 1>,<at 0.776 0 RCIM_Cubic>]}
       BlueCurve {keys [<at 0 0 RCIM_Cubic>,<at 0.489 0.05>]}
       GreenCurve {keys [<at 0 0.4 RCIM_Cubic>,<at 0.489 0.05>]}
       RedCurve {keys [<at 0 0.1333 RCIM_Cubic>,<at 0.489 0.05>]}
      }})
    Grey=&ColorFromCurve(ColorCurve=DataInterface() {propertyValues {
       AlphaCurve {keys [<at 0 3 RCIM_Cubic>,<at 0.7688 0 RCIM_Cubic>]}
       BlueCurve {keys [<at 0 0.25 RCIM_Cubic>,<at 0.4908 0>]}
       GreenCurve {keys [<at 0 0.25 RCIM_Cubic>,<at 0.4908 0>]}
       RedCurve {keys [<at 0 0.25 RCIM_Cubic>,<at 0.4908 0>]}
      }})
    Lime=&ColorFromCurve(ColorCurve=DataInterface() {propertyValues {
       AlphaCurve {keys [<at 0 4 RCIM_Cubic>,<at 0.7634 0>]}
       BlueCurve {keys [<at 0 0 RCIM_Cubic>,<at 0.7616 0>]}
       GreenCurve {keys [<at 0 0.4 RCIM_Cubic>,<at 0.7616 0.2>]}
       RedCurve {keys [<at 0 0.2333 RCIM_Cubic>,<at 0.7616 0.1003>]}
      }})
    Orange=&ColorFromCurve(ColorCurve=DataInterface() {propertyValues {
       AlphaCurve {keys [<at 0 4 RCIM_Cubic>,<at 0.6899 0>]}
       BlueCurve {keys [<at 0 0 RCIM_Cubic>,<at 0.7634 0>]}
       GreenCurve {keys [<at 0 0.1744 RCIM_Cubic>,<at 0.7634 0.1046>]}
       RedCurve {keys [<at 0 0.5 RCIM_Cubic>,<at 0.7634 0.3>]}
      }})
    Pink=&ColorFromCurve(ColorCurve=DataInterface() {propertyValues {
       AlphaCurve {keys [<at 0 4 RCIM_Cubic>,<at 0.7598 0>]}
       BlueCurve {keys [<at 0 0.4688 RCIM_Cubic>,<at 0.6253 0.1953>]}
       GreenCurve {keys [<at 0 0.24 RCIM_Cubic>,<at 0.6253 0.1>]}
       RedCurve {keys [<at 0 0.6 RCIM_Cubic>,<at 0.6253 0.25>]}
      }})
    Purple=&ColorFromCurve(ColorCurve=DataInterface() {propertyValues {
       AlphaCurve {keys [<at 0 4 RCIM_Cubic>,<at 0.758 0>]}
       BlueCurve {keys [<at 0 0.65 RCIM_Cubic>,<at 0.6092 0.5>]}
       GreenCurve {keys [<at 0 0 RCIM_Cubic>,<at 0.6092 0>]}
       RedCurve {keys [<at 0 0.13 RCIM_Cubic>,<at 0.6092 0.1>]}
      }})
    R2 Black:LinearColor=&ColorFromCurve(ColorCurve=DataInterface() {propertyValues {
       AlphaCurve {keys [<at 0 5 RCIM_Cubic>,<at 0.6784 0.7>,<at 0.8882 -0.1>]}
       BlueCurve {keys [{arriveTangent=-0.789, interpMode=RCIM_Cubic, time=0.0604, value=0.05}, <at 0.2123 0.025>]}
       GreenCurve {keys [{arriveTangent=-0.789, interpMode=RCIM_Cubic, time=0.0604, value=0.05}, <at 0.2123 0.025>]}
       RedCurve {keys [{arriveTangent=-4.4991, interpMode=RCIM_Cubic, time=0.0604, value=0.05}, <at 0.2123 0.025>]}
      }})
    R2 BurntSienna:LinearColor=&ColorFromCurve(ColorCurve=DataInterface() {propertyValues {
       AlphaCurve {keys [<at 0 1 RCIM_Cubic>,<at 0.6784 0.7>,<at 0.865 -0.2 RCIM_Cubic>]}
       BlueCurve {keys [<at 0 0.075>, {arriveTangent=-0.0659, interpMode=RCIM_Cubic, leaveTangent=-0.0659, time=0.2902, value=0.024}, <at 0.9945 0.0095>]}
       GreenCurve {keys [<at 0 0.0912>, {arriveTangent=-0.0833, interpMode=RCIM_Cubic, leaveTangent=-0.0833, time=0.2902, value=0.0361}, <at 0.9945 0.0084>]}
       RedCurve {keys [<at 0 0.15>, {arriveTangent=-0.1005, interpMode=RCIM_Cubic, leaveTangent=-0.1005, time=0.2902, value=0.08}, <at 0.9945 0.05>]}
      }})
    R2 Cobalt:LinearColor=&ColorFromCurve(ColorCurve=DataInterface() {propertyValues {
       AlphaCurve {keys [<at 0 1 RCIM_Cubic>,<at 0.5005 0.8>,<at 0.8921 0 RCIM_Cubic>]}
       BlueCurve {keys [<at 0 0.4 RCIM_Cubic>,<at 0.3468 0.25>,<at 0.9228 0.15>]}
       GreenCurve {keys [<at 0 0.094 RCIM_Cubic>,<at 0.3468 0.0925>,<at 0.9228 0.0555>]}
       RedCurve {keys [<at 0 0.06 RCIM_Cubic>,<at 0.3468 0.075>,<at 0.9228 0.045>]}
      }})
    R2 Crimson:LinearColor=&ColorFromCurve(ColorCurve=DataInterface() {propertyValues {
       AlphaCurve {keys [<at 0 1 RCIM_Cubic>, {interpMode=RCIM_Cubic, leaveTangent=-1.2282, time=0.5477, value=1}, <at 0.8844 0>]}
       BlueCurve {keys [<at 0 0.0664 RCIM_Cubic>,<at 0.4583 0.0053>,<at 0.9593 0>]}
       GreenCurve {keys [<at 0 0.06 RCIM_Cubic>,<at 0.4583 0>,<at 0.9593 0>]}
       RedCurve {keys [<at 0 0.3 RCIM_Cubic>,<at 0.4583 0.2>,<at 0.9593 0.1>]}
      }})
    R2 DefaultColor:LinearColor=&ColorFromCurve(ColorCurve=DataInterface() {propertyValues {
       AlphaCurve {keys [<at 0 1 RCIM_Cubic>,<at 0.5619 1>,<at 0.896 0>]}
       BlueCurve {keys [{arriveTangent=-0.789, interpMode=RCIM_Cubic, time=0, value=0.24}, <at 0.2804 0.08>]}
       GreenCurve {keys [{arriveTangent=-0.789, interpMode=RCIM_Cubic, time=0, value=0.3}, <at 0.2804 0.046>]}
       RedCurve {keys [{arriveTangent=-4.4991, interpMode=RCIM_Cubic, time=0, value=0.06}, <at 0.2804 0>]}
      }})
    R2 ForestGreen:LinearColor=&ColorFromCurve(ColorCurve=DataInterface() {propertyValues {
       AlphaCurve {keys [<at 0 1 RCIM_Cubic>, {interpMode=RCIM_Cubic, leaveTangent=-1.2282, time=0.5502, value=1}, <at 0.8869 0>]}
       BlueCurve {keys [<at 0 0.0011 RCIM_Cubic>,<at 0.4777 0.0007>,<at 0.9593 0.0003>]}
       GreenCurve {keys [<at 0 0.25 RCIM_Cubic>,<at 0.4777 0.1>,<at 0.9593 0.0545>]}
       RedCurve {keys [<at 0 0 RCIM_Cubic>,<at 0.4777 0>,<at 0.9593 0>]}
      }})
    R2 Grey:LinearColor=&ColorFromCurve(ColorCurve=DataInterface() {propertyValues {
       AlphaCurve {keys [<at 0 1 RCIM_Cubic>, {interpMode=RCIM_Cubic, leaveTangent=-1.2282, time=0.5502, value=1}, <at 0.9017 0>]}
       BlueCurve {keys [<at 0 0.2 RCIM_Cubic>,<at 0.2293 0.06>,<at 0.9593 0.015>]}
       GreenCurve {keys [<at 0 0.2 RCIM_Cubic>,<at 0.2293 0.06>,<at 0.9593 0.015>]}
       RedCurve {keys [<at 0 0.2 RCIM_Cubic>,<at 0.2293 0.06>,<at 0.9593 0.015>]}
      }})
    R2 Lime:LinearColor=&ColorFromCurve(ColorCurve=DataInterface() {propertyValues {
       AlphaCurve {keys [<at 0 1 RCIM_Cubic>, {interpMode=RCIM_Cubic, leaveTangent=-1.2282, time=0.5697, value=1}, <at 0.8999 0>]}
       BlueCurve {keys [<at 0 0 RCIM_Cubic>,<at 0.2293 0>,<at 0.9593 0>]}
       GreenCurve {keys [<at 0 0.4 RCIM_Cubic>,<at 0.2293 0.2>,<at 0.9593 0.13>]}
       RedCurve {keys [<at 0 0.2543 RCIM_Cubic>,<at 0.2293 0.1359>,<at 0.9593 0.0843>]}
      }})
    R2 Orange:LinearColor=&ColorFromCurve(ColorCurve=DataInterface() {propertyValues {
       AlphaCurve {keys [<at 0 1 RCIM_Cubic>, {interpMode=RCIM_Cubic, leaveTangent=-1.2088, time=0.6564, value=0.7}, <at 0.9232 0>]}
       BlueCurve {keys [<at 0 0 RCIM_Cubic>,<at 0.2036 0.009>,<at 0.8178 0.0075>]}
       GreenCurve {keys [<at 0 0.1352 RCIM_Cubic>,<at 0.2036 0.0694>,<at 0.8178 0.0578>]}
       RedCurve {keys [<at 0 0.4 RCIM_Cubic>,<at 0.2036 0.18>,<at 0.8178 0.15>]}
      }})
    R2 Pink:LinearColor=&ColorFromCurve(ColorCurve=DataInterface() {propertyValues {
       AlphaCurve {keys [<at 0 1 RCIM_Cubic>, {interpMode=RCIM_Cubic, leaveTangent=-1, time=0.6331, value=0.8}, <at 0.918 0>]}
       BlueCurve {keys [<at 0 0.34 RCIM_Cubic>,<at 0.2293 0.17>,<at 0.9632 0.1105>]}
       GreenCurve {keys [<at 0 0.2 RCIM_Cubic>,<at 0.2293 0.1>,<at 0.9632 0.065>]}
       RedCurve {keys [<at 0 0.4 RCIM_Cubic>,<at 0.2293 0.2>,<at 0.9632 0.13>]}
      }})
    R2 Purple:LinearColor=&ColorFromCurve(ColorCurve=DataInterface() {propertyValues {
       AlphaCurve {keys [<at 0 1 RCIM_Cubic>, {interpMode=RCIM_Cubic, leaveTangent=-1.2904, time=0.6292, value=0.8}, <at 0.9154 0>]}
       BlueCurve {keys [<at 0 0.3 RCIM_Cubic>,<at 0.2341 0.2>,<at 0.9616 0.15>]}
       GreenCurve {keys [<at 0 0.12 RCIM_Cubic>,<at 0.2341 0.04>,<at 0.9616 0.03>]}
       RedCurve {keys [<at 0 0.135 RCIM_Cubic>,<at 0.2341 0.0667>,<at 0.9616 0.05>]}
      }})
    R2 Saffron:LinearColor=&ColorFromCurve(ColorCurve=DataInterface() {propertyValues {
       AlphaCurve {keys [<at 0 1 RCIM_Cubic>, {interpMode=RCIM_Cubic, leaveTangent=-1.0832, time=0.5645, value=0.8}, <at 0.9038 0>]}
       BlueCurve {keys [<at 0 0.0069 RCIM_Cubic>,<at 0.2293 0.0028>,<at 0.9593 0>]}
       GreenCurve {keys [<at 0 0.2 RCIM_Cubic>,<at 0.2293 0.17>,<at 0.9593 0.1301>]}
       RedCurve {keys [<at 0 0.1982 RCIM_Cubic>,<at 0.2293 0.17>,<at 0.9593 0.1301>]}
      }})
    R2 SkyBlue:LinearColor=&ColorFromCurve(ColorCurve=DataInterface() {propertyValues {
       AlphaCurve {keys [<at 0 0.5 RCIM_Cubic>, {interpMode=RCIM_Cubic, leaveTangent=-0.6521, time=0.5658, value=0.9}, <at 0.8934 0>]}
       BlueCurve {keys [<at 0 0.3 RCIM_Cubic>,<at 0.9593 0.14>]}
       GreenCurve {keys [<at 0 0.15 RCIM_Cubic>,<at 0.9593 0.1166>]}
       RedCurve {keys [<at 0 0 RCIM_Cubic>,<at 0.9593 0.07>]}
      }})
    R2 TitaniumWhite:LinearColor=&ColorFromCurve(ColorCurve=DataInterface() {propertyValues {
       AlphaCurve {keys [<at 0 1 RCIM_Cubic>, {interpMode=RCIM_Cubic, leaveTangent=-1.1915, time=0.5865, value=0.7}, <at 0.9477 -0.1>]}
       BlueCurve {keys [<at 0.4317 0.29 RCIM_Cubic>,<at 0.9578 0.2>]}
       GreenCurve {keys [<at 0.4317 0.29 RCIM_Cubic>,<at 0.9578 0.2>]}
       RedCurve {keys [<at 0.4317 0.29 RCIM_Cubic>,<at 0.9578 0.2>]}
      }})
    R3 Black:LinearColor=&ColorFromCurve(ColorCurve=DataInterface() {propertyValues {
       AlphaCurve {keys [<at 0 15 RCIM_Cubic>]}
       BlueCurve {keys [<at 0 0 RCIM_Cubic>]}
       GreenCurve {keys [<at 0 0 RCIM_Cubic>]}
       RedCurve {keys [<at 0 0 RCIM_Cubic>]}
      }})
    R3 BurntSienna:LinearColor=&ColorFromCurve(ColorCurve=DataInterface() {propertyValues {
       AlphaCurve {keys [<at 0 7 RCIM_Cubic>]}
       BlueCurve {keys [<at 0 0.0619 RCIM_Cubic>,<at 0.2346 0.0131 RCIM_Cubic>]}
       GreenCurve {keys [<at 0 0.1565 RCIM_Cubic>,<at 0.2346 0.0282 RCIM_Cubic>]}
       RedCurve {keys [<at 0 0.5 RCIM_Cubic>,<at 0.2346 0.08 RCIM_Cubic>]}
      }})
    R3 Cobalt:LinearColor=&ColorFromCurve(ColorCurve=DataInterface() {propertyValues {
       AlphaCurve {keys [<at 0 20 RCIM_Cubic>,<at 0.9944 0.5>]}
       BlueCurve {keys [<at 0 0.15 RCIM_Cubic>]}
       GreenCurve {keys [<at 0 0.0263 RCIM_Cubic>]}
       RedCurve {keys [<at 0 0.015 RCIM_Cubic>]}
      }})
    R3 Crimson:LinearColor=&ColorFromCurve(ColorCurve=DataInterface() {propertyValues {
       AlphaCurve {keys [<at 0 10 RCIM_Cubic>]}
       BlueCurve {keys [<at 0 0.0015 RCIM_Cubic>]}
       GreenCurve {keys [<at 0 0 RCIM_Cubic>]}
       RedCurve {keys [<at 0 0.11 RCIM_Cubic>]}
      }})
    R3 DefaultColor:LinearColor=&ColorFromCurve(ColorCurve=DataInterface() {propertyValues {
       AlphaCurve {keys [<at 0 1 RCIM_Cubic>,<at 0.842 0 RCIM_Cubic>]}
       BlueCurve {keys [<at 0 1 RCIM_Cubic>,<at 0.8808 1>]}
       GreenCurve {keys [<at 0 0.8959 RCIM_Cubic>,<at 0.8808 0.9271>]}
       RedCurve {keys [<at 0 0.0009 RCIM_Cubic>,<at 0.8808 0.3>]}
      }})
    R3 ForestGreen:LinearColor=&ColorFromCurve(ColorCurve=DataInterface() {propertyValues {
       AlphaCurve {keys [<at 0 10 RCIM_Cubic>]}
       BlueCurve {keys [<at 0 0.0161 RCIM_Cubic>,<at 0.1988 0.001>]}
       GreenCurve {keys [<at 0 1 RCIM_Cubic>,<at 0.1988 0.06>]}
       RedCurve {keys [<at 0 0 RCIM_Cubic>,<at 0.1988 0>]}
      }})
    R3 Grey:LinearColor=&ColorFromCurve(ColorCurve=DataInterface() {propertyValues {
       AlphaCurve {keys [<at 0 10 RCIM_Cubic>]}
       BlueCurve {keys [<at 0.0672 10 RCIM_Cubic>,<at 0.2165 0.09 RCIM_Cubic>]}
       GreenCurve {keys [<at 0.0672 10 RCIM_Cubic>,<at 0.2165 0.09 RCIM_Cubic>]}
       RedCurve {keys [<at 0.0672 10 RCIM_Cubic>,<at 0.2165 0.09 RCIM_Cubic>]}
      }})
    R3 Lime:LinearColor=&ColorFromCurve(ColorCurve=DataInterface() {propertyValues {
       AlphaCurve {keys [<at 0 10 RCIM_Cubic>]}
       BlueCurve {keys [<at 0 0 RCIM_Cubic>,<at 0.2149 0>]}
       GreenCurve {keys [<at 0 1 RCIM_Cubic>,<at 0.2149 0.15>]}
       RedCurve {keys [<at 0 0.6667 RCIM_Cubic>,<at 0.2149 0.1>]}
      }})
    R3 Orange:LinearColor=&ColorFromCurve(ColorCurve=DataInterface() {propertyValues {
       AlphaCurve {keys [<at 0 10 RCIM_Cubic>]}
       BlueCurve {keys [<at 0 0.0022 RCIM_Cubic>,<at 0.2324 0.0003>]}
       GreenCurve {keys [<at 0 0.1083 RCIM_Cubic>,<at 0.2324 0.0162>]}
       RedCurve {keys [<at 0 0.4 RCIM_Cubic>,<at 0.2324 0.06>]}
      }})
    R3 Pink:LinearColor=&ColorFromCurve(ColorCurve=DataInterface() {propertyValues {
       AlphaCurve {keys [<at 0 10 RCIM_Cubic>]}
       BlueCurve {keys [<at 0 0.4339 RCIM_Cubic>,<at 0.223 0.0868>]}
       GreenCurve {keys [<at 0 0.0755 RCIM_Cubic>,<at 0.223 0.0151>]}
       RedCurve {keys [<at 0 0.6 RCIM_Cubic>,<at 0.223 0.12>]}
      }})
    R3 Purple:LinearColor=&ColorFromCurve(ColorCurve=DataInterface() {propertyValues {
       AlphaCurve {keys [<at 0 15 RCIM_Cubic>,<at 1 0>]}
       BlueCurve {keys [<at 0 0.35 RCIM_Cubic>,<at 0.2378 0.18>]}
       GreenCurve {keys [<at 0 0.0455 RCIM_Cubic>,<at 0.2378 0.0234>]}
       RedCurve {keys [<at 0 0.0962 RCIM_Cubic>,<at 0.2378 0.0495>]}
      }})
    R3 Saffron:LinearColor=&ColorFromCurve(ColorCurve=DataInterface() {propertyValues {
       AlphaCurve {keys [<at 0.001 10 RCIM_Cubic>,<at 1 0>]}
       BlueCurve {keys [<at 0 0 RCIM_Cubic>,<at 0.2311 0>]}
       GreenCurve {keys [<at 0 1 RCIM_Cubic>,<at 0.2311 0.18>]}
       RedCurve {keys [<at 0 0.9076 RCIM_Cubic>,<at 0.2311 0.1634>]}
      }})
    R3 SkyBlue:LinearColor=&ColorFromCurve(ColorCurve=DataInterface() {propertyValues {
       AlphaCurve {keys [<at 0 8 RCIM_Cubic>,<at 1 0 RCIM_Cubic>]}
       BlueCurve {keys [<at 0 1 RCIM_Cubic>,<at 0.2486 0.17>]}
       GreenCurve {keys [<at 0 0.7898 RCIM_Cubic>,<at 0.2486 0.1343>]}
       RedCurve {keys [<at 0 0 RCIM_Cubic>,<at 0.2486 0>]}
      }})
    R3 TitaniumWhite:LinearColor=&ColorFromCurve(ColorCurve=DataInterface() {propertyValues {
       AlphaCurve {keys [<at 0 10 RCIM_Cubic>,<at 1 1>]}
       BlueCurve {keys [<at 0 1 RCIM_Cubic>,<at 0.2257 0.2>]}
       GreenCurve {keys [<at 0 1 RCIM_Cubic>,<at 0.2257 0.2>]}
       RedCurve {keys [<at 0 1 RCIM_Cubic>,<at 0.2257 0.2>]}
      }})
    Saffron=&ColorFromCurve(ColorCurve=DataInterface() {propertyValues {
       AlphaCurve {keys [<at 0 4 RCIM_Cubic>,<at 0.7598 0>]}
       BlueCurve {keys [<at 0 0 RCIM_Cubic>,<at 1 0>]}
       GreenCurve {keys [<at 0 0.4 RCIM_Cubic>,<at 1 0.2>]}
       RedCurve {keys [<at 0 0.3813 RCIM_Cubic>,<at 1 0.1906>]}
      }})
    SkyBlue=&ColorFromCurve(ColorCurve=DataInterface() {propertyValues {
       AlphaCurve {keys [<at 0 4 RCIM_Cubic>,<at 0.767 0>]}
       BlueCurve {keys [<at 0 0.5 RCIM_Cubic>,<at 0.7706 0.05>]}
       GreenCurve {keys [<at 0 0.3333 RCIM_Cubic>,<at 0.7706 0.0417>]}
       RedCurve {keys [<at 0 0 RCIM_Cubic>,<at 0.7706 0>]}
      }})
    TitaniumWhite=&ColorFromCurve(ColorCurve=DataInterface() {propertyValues {
       AlphaCurve {keys [<at 0 4 RCIM_Cubic>,<at 0.7423 0>]}
       BlueCurve {keys [<at 0 0.4 RCIM_Cubic>,<at 0.645 0.3>]}
       GreenCurve {keys [<at 0 0.4 RCIM_Cubic>,<at 0.645 0.3>]}
       RedCurve {keys [<at 0 0.4 RCIM_Cubic>,<at 0.645 0.3>]}
      }})
    Use R2 Painted Overrides=true
    Use R3 Painted Overrides=true
   )
   &ScaleRibbonWidth(Ribbon Width Scale=&FloatFromCurve(FloatCurve=DataInterface() {propertyValues {Curve {keys [{arriveTangent=0.5755, interpMode=RCIM_Cubic, leaveTangent=0.5755, tangentMode=RCTM_User, time=0, value=0}, <at 1 0.8>]}}}))
   /SetVariables(
    Particles.InnerWidthB:NiagaraFloat=&FloatFromCurve(FloatCurve=DataInterface() {propertyValues {Curve {keys [<at 0 0>,<at 1 1>]}}}, Scale Curve=200)
    Particles.RaysWidthA:NiagaraFloat=&Multiply_Float(
     A=30
     B=&FloatFromCurve(FloatCurve=DataInterface() {propertyValues {Curve {keys [{tangentMode=RCTM_User, time=0, value=0}, <at 0.15 0.25>, <at 0.5 1.3>]}}})
    )
   ) {isSetParameters=true}
   /SetVariables(Particles.PositionRays:NiagaraPosition=&AddVectorToPosition(
     Position=$Particles.Position
     Vector=&TransformVector(
      Vector=&MakeVector(
       X=0
       Y=0
       Z=&Sine(
        Normalized Angle=&Add_Float(
         A=&Multiply_Float(A=$Particles.NormalizedAge, B=-1)
         B=&Multiply_Float(A=$Emitter.LoopedAge, B=1)
        )
        Period=0.75
        Scale=&FloatFromCurve(
         FloatCurve=DataInterface() {propertyValues {Curve {
            defaultValue=0
            keys [
             {interpMode=RCIM_Cubic, tangentMode=RCTM_User, time=0, value=0}
             {interpMode=RCIM_Cubic, tangentMode=RCTM_User, time=0.2, value=10}
             <at 0.4 -10 RCIM_Cubic>
             <at 0.6 10 RCIM_Cubic>
             <at 0.7 -10 RCIM_Cubic>
             {arriveTangent=0.5014, interpMode=RCIM_Cubic, leaveTangent=0.5014, tangentMode=RCTM_User, time=0.8, value=10}
             {tangentMode=RCTM_User, time=1, value=0}
            ]
           }}}
         Scale Curve=0.2
        )
        Bias=1
       )
      )
      Destination Space=::Simulation
      Apply Scale=true
     )
    )) {isSetParameters=true}
  ]
  rendererClasses [&NiagaraRibbonRendererProperties]
  renderers [
   {rendererClass=&NiagaraRibbonRendererProperties}
   {rendererClass=&NiagaraRibbonRendererProperties, rendererIndex=1}
   {rendererClass=&NiagaraRibbonRendererProperties, rendererIndex=2}
  ]
 }
}
summary {
 systemName=NS_DriftTrail_techno_MP
 emitters [
  {emitterName=NE_VC_DriftTrail_WheelWrap_Parent, rendererClasses [&NiagaraMeshRendererProperties]}
  {emitterName=TrailPlasmaEnergy, rendererClasses [&NiagaraRibbonRendererProperties]}
 ]
 userVariables [
  User.Alpha_line {type=NiagaraFloat, defaultValue=10, description=""}
  User.alpharay {type=NiagaraFloat, defaultValue=3, description=""}
  User.alphawr {type=NiagaraFloat, defaultValue=3, description=""}
  User.bIsFrontendPreview {type=NiagaraBool, defaultValue=true}
  User.bIsPainted {type=NiagaraBool}
  User.black {type=NiagaraFloat, defaultValue=1, description=""}
  User.ColorVariantID {type=NiagaraInt32}
  User.ColorWhite {type=NiagaraFloat, defaultValue=0, description=""}
  User.Degradadoline {type=NiagaraFloat, defaultValue=0, description=""}
  User.Intensity {type=NiagaraFloat, defaultValue=0.8, description=""}
  User.offset {type=Vector3f, description="offset from wheels(set through cosmetic template)", defaultValue {struct=Vector3f, value=<0 0 0>}}
  User.offset001 {type=NiagaraPosition, description="", defaultValue {}}
  User.PaintedColor {type=LinearColor, description="", defaultValue {struct=LinearColor, value=<rgba 1 1 1 1>}}
  User.PaintedColor_4 {type=LinearColor, description="", defaultValue {struct=LinearColor, value=<rgba 1 1 1 1>}}
  User.PaintedColor2 {type=LinearColor, description="", defaultValue {struct=LinearColor, value=<rgba 1 1 1 1>}}
  User.PaintedScalar {type=NiagaraFloat, defaultValue=1, description=""}
  User.Power_lightning {type=NiagaraFloat, defaultValue=1, description=""}
  User.RibbonColor {type=LinearColor, description="", defaultValue {struct=LinearColor, value=<rgba 0.805 8 0 1>}}
  User.TireOffsetPosition {type=Vector3f, description="", defaultValue {struct=Vector3f, value=<-30 0 0>}}
  User.VectorScaleMax {type=Vector3f, description="", defaultValue {struct=Vector3f, value=<1 1 1>}}
  User.VectorScaleMin {type=Vector3f, description="", defaultValue {struct=Vector3f, value=<0.9 0.9 0.9>}}
  User.VehicleFlags {type=NiagaraInt32, defaultValue=806289408}
  User.Wheel_Radius_Array {
   type=NiagaraDataInterfaceArrayFloat
   defaultValue {
    struct=NiagaraExt_VariableValue_DataInterface
    value {
     dataInterface=&self:NiagaraDataInterfaceArrayFloat_0
     dataInterfaceClass=&NiagaraDataInterfaceArrayFloat
    }
   }
  }
  User.Wheel_Width_Array {
   type=NiagaraDataInterfaceArrayFloat
   defaultValue {
    struct=NiagaraExt_VariableValue_DataInterface
    value {
     dataInterface=&self:NiagaraDataInterfaceArrayFloat_1
     dataInterfaceClass=&NiagaraDataInterfaceArrayFloat
    }
   }
  }
  User.WheelSize {type=NiagaraFloat, defaultValue=0.1, description=""}
  User.white {type=NiagaraFloat, defaultValue=1, description=""}
 ]
}
